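# FortResourceJobBrief is a quest brief widget for the job board, a child of FortQuestBrief that overrides how quest information is displayed.
# When quest information arrives, it forwards the objective, rarity, category and reward data to the parent handler.
# It then overwrites the Title text with a localized "SCAVENGING" label.
# It adds no widgets or variables of its own.
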WidgetBlueprint /JobBoard/Widgets/FortResourceJobBrief
graphs [EventGraph {nodes [
   n1 = Event OnQuestInformationRecieved!() -> (
    OutputDelegate
    then
    Objective:"Text (by ref)"
    RarityData:"Fort Rarity Item Data Structure (by ref)"
    CategoryTag:"Gameplay Tag Structure (by ref)"
    XPReward:int=0
    MoneyReward:int=0
    TokenReward:int=0
    TokenRewardIcon:"Texture 2D Soft Object Reference (by ref)"
   ) {is=title K2Node_Event}
   > Parent: OnQuestInformationRecieved!(
    execute
    Objective:"Text (by ref)"=@n1.Objective
    RarityData:"Fort Rarity Item Data Structure (by ref)"=@n1.RarityData
    CategoryTag:"Gameplay Tag Structure (by ref)"=@n1.CategoryTag
    XPReward:int=@n1.XPReward
    MoneyReward:int=@n1.MoneyReward
    TokenReward:int=@n1.TokenReward
    TokenRewardIcon:"Texture 2D Soft Object Reference (by ref)"=@n1.TokenRewardIcon
   ) -> (then) {is=title K2Node_CallParentFunction}
   > "SetText (Text)"(self=Get Title() -> (Title:Common Text Object Reference), InText="LOCTEXT(\"SCAVENGING\")")
   Event PreConstruct()
   Event Construct()
   Event Tick()
  ]}]
variables []
widgets {
 info {
  inheritedWidgetCount=0
  parentClass=/FortniteConversationUI/Conversation/FortQuestBrief.FortQuestBrief_C
  rootWidgetClass=null
  widgetCount=0
 }
 widgets []
}
